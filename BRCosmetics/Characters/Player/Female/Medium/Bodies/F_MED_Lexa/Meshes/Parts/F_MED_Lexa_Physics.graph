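# Physics asset for the Lexa female medium player character, defining bodies and constraints for dynamic hair.
# Small base spheres root chains of mostly box bodies for bangs, strands, side hair and back hair.
# Swing limits are tight on the bangs and widest at the back, with uniformly tight twist.
# Unconstrained capsules on head, neck, upper spine and right clavicle provide body collision shapes for the hair.
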PhysicsAsset /BRCosmetics/Characters/Player/Female/Medium/Bodies/F_MED_Lexa/Meshes/Parts/F_MED_Lexa_Physics
bodies {
 clavicle_r [{shapeType=Capsule, length=9.0787, radius=5.9555, center=<-7.5616 0.39 1.6489>, rotation=<rot 80.3819 12.9295 0>}]
 dyn_bang_01_l [{shapeType=Box, extentX=2.7071, extentY=4.1089, extentZ=1.0061, center=<1.3688 0 0>}]
 dyn_bang_01_r [{shapeType=Box, extentX=2.1223, extentY=4.1069, extentZ=1.0126, center=<0.9981 0 0>}]
 dyn_bang_base_l [{shapeType=Sphere, radius=0.5}]
 dyn_bang_base_r [{shapeType=Sphere, radius=0.5}]
 dyn_bang_parent_l [{shapeType=Box, extentX=3.1716, extentY=4.1969, extentZ=0.5006, center=<1.4317 0 0.0001>}]
 dyn_bang_parent_r [{shapeType=Box, extentX=3.2737, extentY=3.4081, extentZ=0.6579, center=<1.4927 0 0>}]
 dyn_bang_side_01_r [{shapeType=Box, extentX=2.0475, extentY=4.5123, extentZ=0.9967, center=<0.9358 0 0>}]
 dyn_bang_side_02_r [{shapeType=Box, extentX=5.3817, extentY=5.122, extentZ=1.687, center=<0.4409 0 0>}]
 dyn_bang_side_03_r [{shapeType=Box, extentX=2.5631, extentY=3.7744, extentZ=1.2883, center=<1.2189 0 0.0001>}]
 dyn_bang_side_04_r [{shapeType=Box, extentX=2.085, extentY=1.8955, extentZ=0.9637, center=<0.9229 0 0>}]
 dyn_bang_side_base_r [{shapeType=Sphere, radius=0.5}]
 dyn_bang_side_parent_r [{shapeType=Box, extentX=5.6895, extentY=4.6551, extentZ=0.9637, center=<2.837 0.0001 0>}]
 dyn_bang_strand_01_l [{shapeType=Box, extentX=1.9319, extentY=1.3579, extentZ=1.5581, center=<0.8801 0 -0.0002>}]
 dyn_bang_strand_base_l [{shapeType=Sphere, radius=0.5}]
 dyn_bang_strand_base_r [{shapeType=Sphere, radius=0.5}]
 dyn_bang_strand_parent_l [{shapeType=Box, extentX=1.851, extentY=1.7075, extentZ=0.5006, center=<0.9458 0 0>}]
 dyn_bang_strand_parent_r [{shapeType=Box, extentX=2.8391, extentY=2.1786, extentZ=0.9849, center=<1.1815 0 0>}]
 dyn_hair_01_back_l [{shapeType=Box, extentX=4.0732, extentY=5.2392, extentZ=1.2673, center=<2.0444 0 0>}]
 dyn_hair_01_back_mid [{shapeType=Box, extentX=5.0406, extentY=6.656, extentZ=2.3437, center=<2.6152 0 0>}]
 dyn_hair_01_back_r [{shapeType=Box, extentX=3.4078, extentY=6.8679, extentZ=1.9492, center=<1.6371 0 0>, rotation=<rot -24.0614 0 0>}]
 dyn_hair_02_back_l [{shapeType=Box, extentX=5.2985, extentY=5.4079, extentZ=1.01, center=<2.576 0 0>}]
 dyn_hair_02_back_mid [{shapeType=Box, extentX=4.4005, extentY=6.8075, extentZ=1.298, center=<2.2006 0 0>}]
 dyn_hair_02_back_r [{shapeType=Box, extentX=3.8469, extentY=7.1505, extentZ=2.8446, center=<1.7608 0 0>}]
 dyn_hair_03_back_r [{shapeType=Box, extentX=4.7214, extentY=7.0386, extentZ=2.4254, center=<1.8363 0 0>}]
 dyn_hair_base_back_l [{shapeType=Sphere, radius=0.5}]
 dyn_hair_base_back_mid [{shapeType=Sphere, radius=0.5}]
 dyn_hair_base_back_r [{shapeType=Sphere, radius=0.5}]
 dyn_hair_offset [{shapeType=Sphere, radius=0.5001}]
 dyn_hair_parent_back_l [{shapeType=Box, extentX=3.7837, extentY=5.237, extentZ=1.2449, center=<1.7703 0 0>}]
 dyn_hair_parent_back_mid [{shapeType=Box, extentX=4.869, extentY=6.2105, extentZ=1.4702, center=<2.3293 0 -0.0001>}]
 dyn_hair_parent_back_r [{shapeType=Box, extentX=3.5081, extentY=6.1655, extentZ=2.4357, center=<1.4684 0 0>}]
 dyn_hair_side_01_l [{shapeType=Box, extentX=4.4372, extentY=5.7101, extentZ=3.2801, center=<2.2933 0 0>}]
 dyn_hair_side_01_r [{shapeType=Box, extentX=4.1206, extentY=4.1836, extentZ=2.4358, center=<1.721 0 0>}]
 dyn_hair_side_02_l [{shapeType=Box, extentX=4.8275, extentY=5.2774, extentZ=1.7315, center=<2.2699 0 0>}]
 dyn_hair_side_02_r [{shapeType=Box, extentX=2.6041, extentY=5.3596, extentZ=2.3215, center=<1.2628 0 0>}]
 dyn_hair_side_03_l [{shapeType=Box, extentX=2.3821, extentY=5.6936, extentZ=2.3087, center=<1.2269 0 0.0001>}]
 dyn_hair_side_03_r [{shapeType=Capsule, length=4.9824, radius=1.6808, center=<0.8839 -0.0193 -0.6728>, rotation=<rot 0 0 -88.358>}]
 dyn_hair_side_base_l [{shapeType=Sphere, radius=0.5}]
 dyn_hair_side_base_r [{shapeType=Sphere, radius=0.5}]
 dyn_hair_side_parent_l [{shapeType=Box, extentX=4.2028, extentY=4.5458, extentZ=1.7029, center=<2.1158 0 0>}]
 dyn_hair_side_parent_r [{shapeType=Box, extentX=4.4738, extentY=4.7329, extentZ=1.9553, center=<2.0751 0 0>}]
 head [{shapeType=Capsule, length=8.9116, radius=6.6529, center=<1.8321 0.7769 0>, rotation=<rot 90 0 -15.7925>}]
 neck_02 [{shapeType=Capsule, length=5.7275, radius=5.5386, center=<-4.9482 2.5677 0>, rotation=<rot 90 0 0>}]
 spine_05 [
  {shapeType=Capsule, length=13.1793, radius=6.9788, center=<-0.1373 4.3722 -10.8914>, rotation=<rot 88.6325 -179.9996 -133.7865>}
  {shapeType=Capsule, length=9.8298, radius=7.8857, center=<3.3259 -1.3783 -0.0001>, rotation=<rot 90 0 -16.4221>}
 ]
}
constraints [
 {bone1Name=dyn_bang_side_parent_r, bone2Name=dyn_bang_side_base_r, swing1LimitDegrees=5, swing2LimitDegrees=3, twistLimitDegrees=3}
 {bone1Name=dyn_bang_side_01_r, bone2Name=dyn_bang_side_parent_r, swing1LimitDegrees=10, swing2LimitDegrees=5, twistLimitDegrees=3}
 {bone1Name=dyn_bang_side_02_r, bone2Name=dyn_bang_side_01_r, swing1LimitDegrees=10, swing2LimitDegrees=5, twistLimitDegrees=3}
 {bone1Name=dyn_bang_side_03_r, bone2Name=dyn_bang_side_02_r, swing1LimitDegrees=10, swing2LimitDegrees=10, twistLimitDegrees=3}
 {bone1Name=dyn_bang_side_04_r, bone2Name=dyn_bang_side_02_r, swing1LimitDegrees=3, swing2LimitDegrees=3, twistLimitDegrees=3}
 {bone1Name=dyn_bang_strand_parent_r, bone2Name=dyn_bang_strand_base_r, swing1LimitDegrees=10, swing2LimitDegrees=6, twistLimitDegrees=3}
 {bone1Name=dyn_bang_parent_r, bone2Name=dyn_bang_base_r, swing1LimitDegrees=5, swing2LimitDegrees=5, twistLimitDegrees=3}
 {bone1Name=dyn_bang_01_r, bone2Name=dyn_bang_parent_r, swing1LimitDegrees=10, swing2LimitDegrees=10, twistLimitDegrees=3}
 {bone1Name=dyn_bang_parent_l, bone2Name=dyn_bang_base_l, swing1LimitDegrees=5, swing2LimitDegrees=5, twistLimitDegrees=3}
 {bone1Name=dyn_bang_01_l, bone2Name=dyn_bang_parent_l, swing1LimitDegrees=10, swing2LimitDegrees=10, twistLimitDegrees=3}
 {bone1Name=dyn_bang_strand_01_l, bone2Name=dyn_bang_strand_parent_l, swing1LimitDegrees=10, swing2LimitDegrees=10, twistLimitDegrees=3}
 {bone1Name=dyn_bang_strand_parent_l, bone2Name=dyn_bang_strand_base_l, swing1LimitDegrees=5, swing2LimitDegrees=5, twistLimitDegrees=3}
 {bone1Name=dyn_hair_side_parent_l, bone2Name=dyn_hair_side_base_l, swing1LimitDegrees=10, swing2LimitDegrees=10, twistLimitDegrees=3}
 {bone1Name=dyn_hair_side_01_l, bone2Name=dyn_hair_side_parent_l, swing1LimitDegrees=20, swing2LimitDegrees=20, twistLimitDegrees=3}
 {bone1Name=dyn_hair_side_02_l, bone2Name=dyn_hair_side_01_l, swing1LimitDegrees=20, swing2LimitDegrees=20, twistLimitDegrees=3}
 {bone1Name=dyn_hair_side_03_l, bone2Name=dyn_hair_side_02_l, swing1LimitDegrees=20, swing2LimitDegrees=20, twistLimitDegrees=3}
 {bone1Name=dyn_hair_side_parent_r, bone2Name=dyn_hair_side_base_r, swing1LimitDegrees=10, swing2LimitDegrees=20, twistLimitDegrees=3}
 {bone1Name=dyn_hair_side_01_r, bone2Name=dyn_hair_side_parent_r, swing1LimitDegrees=15, swing2LimitDegrees=15, twistLimitDegrees=3}
 {bone1Name=dyn_hair_side_02_r, bone2Name=dyn_hair_side_01_r, swing1LimitDegrees=15, swing2LimitDegrees=15, twistLimitDegrees=3}
 {bone1Name=dyn_hair_side_03_r, bone2Name=dyn_hair_side_02_r, swing1LimitDegrees=20, swing2LimitDegrees=20, twistLimitDegrees=3}
 {bone1Name=dyn_hair_parent_back_mid, bone2Name=dyn_hair_base_back_mid, swing1LimitDegrees=30, swing2LimitDegrees=50, twistLimitDegrees=3}
 {bone1Name=dyn_hair_01_back_mid, bone2Name=dyn_hair_parent_back_mid, swing1LimitDegrees=40, swing2LimitDegrees=45, twistLimitDegrees=3}
 {bone1Name=dyn_hair_02_back_mid, bone2Name=dyn_hair_01_back_mid, swing1LimitDegrees=40, swing2LimitDegrees=45, twistLimitDegrees=3}
 {bone1Name=dyn_hair_parent_back_l, bone2Name=dyn_hair_base_back_l, swing1LimitDegrees=30, swing2LimitDegrees=50, twistLimitDegrees=3}
 {bone1Name=dyn_hair_01_back_l, bone2Name=dyn_hair_parent_back_l, swing1LimitDegrees=40, swing2LimitDegrees=45, twistLimitDegrees=3}
 {bone1Name=dyn_hair_02_back_l, bone2Name=dyn_hair_01_back_l, swing1LimitDegrees=40, swing2LimitDegrees=45, twistLimitDegrees=3}
 {bone1Name=dyn_hair_parent_back_r, bone2Name=dyn_hair_base_back_r, swing1LimitDegrees=30, swing2LimitDegrees=50, twistLimitDegrees=3}
 {bone1Name=dyn_hair_01_back_r, bone2Name=dyn_hair_parent_back_r, swing1LimitDegrees=40, swing2LimitDegrees=45, twistLimitDegrees=3}
 {bone1Name=dyn_hair_02_back_r, bone2Name=dyn_hair_01_back_r, swing1LimitDegrees=45, swing2LimitDegrees=45, twistLimitDegrees=3}
 {bone1Name=dyn_hair_03_back_r, bone2Name=dyn_hair_02_back_r, swing1LimitDegrees=45, swing2LimitDegrees=40, twistLimitDegrees=3}
]
