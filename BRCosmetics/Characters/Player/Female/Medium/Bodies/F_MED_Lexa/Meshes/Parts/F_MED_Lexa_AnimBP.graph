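# Animation Blueprint for a part mesh of the female player character Lexa, driving hair motion and transformation materials.
# It copies the parent mesh's pose, additively rotates two right-side hair bones, then runs RigidBody physics relative to spine_05.
# An interpolated alpha enables that rotation when head twist against spine_05 crosses a threshold during targeting.
# It also feeds the body's Lexa_TransformationCurve and a head socket location into material parameters.
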AnimBlueprint /BRCosmetics/Characters/Player/Female/Medium/Bodies/F_MED_Lexa/Meshes/Parts/F_MED_Lexa_AnimBP
target_skeleton=&/F_MED_Lexa_Skeleton
graphs [
 AnimGraph {nodes [
   n1 = Get Hair_Alpha() -> (Hair_Alpha:double)
   n2 = "Transform (Modify) Bone - Bone: dyn_hair_side_base_r"(
    ComponentPose=Local To Component(LocalPose=Copy Pose From Mesh() {bCopyCurves=true, bUseAttachedParent=true})
    Alpha=@n1
    Translation="0,0,0"
    Rotation="-35,-45,20"
    Scale="1,1,1"
   ) {
    rotationMode=BMM_Additive
    rotationSpace=BCS_BoneSpace
    alphaBoolBlend {blendOption=Linear}
    alphaScaleBiasClamp {bInterpResult=true, interpSpeedDecreasing=4, interpSpeedIncreasing=6}
    boneToModify {boneName=dyn_hair_side_base_r}
    rotation=<rot -35 -45 20>
   }
   n3 = "Transform (Modify) Bone - Bone: dyn_hair_base_back_r"(ComponentPose=@n2, Alpha=@n1, Translation="0,0,0", Rotation="-30,-25,20", Scale="1,1,1") {
    rotationMode=BMM_Additive
    rotationSpace=BCS_BoneSpace
    alphaBoolBlend {blendOption=Linear}
    alphaScaleBiasClamp {bInterpResult=true, interpSpeedDecreasing=4, interpSpeedIncreasing=6}
    boneToModify {boneName=dyn_hair_base_back_r}
    rotation=<rot -30 -25 20>
   }
   n4 = RigidBody(ComponentPose=@n3) {
    bClampLinearTranslationLimitToRefPose=true
    bOverrideWorldGravity=true
    overridePhysicsAsset=&/F_MED_Lexa_Physics
    simulationSpace=BaseBoneSpace
    alphaBoolBlend {blendOption=Linear}
    alphaScaleBiasClamp {bInterpResult=true, interpSpeedDecreasing=4, interpSpeedIncreasing=6}
    baseBoneRef {boneName=spine_05}
    componentAppliedLinearAccClamp=<3000 3000 3000>
    componentLinearAccScale=<0.3 0.3 0.32>
    componentLinearVelScale=<0.3 0.3 0.32>
    overrideWorldGravity {z=-980}
   }
   Output Pose(Result=Component To Local(ComponentPose=@n4)) {layerGroup=None}
  ]}
 EventGraph {nodes [
   Event BlueprintInitializeAnimation()
   > n1 = Cast To PlayerPawn_Athena(Object=TryGetPawnOwner())
   > n2 = Set Pawn Owner(Pawn Owner:Player Pawn Athena Object Reference=@n1) -> (Output_Get:Player Pawn Athena Object Reference)
   > Set Body_Mesh(Body_Mesh=GetSkeletalMeshForPartType(self=@n2))
   n3 = Event BlueprintPostEvaluateAnimation() {is=title K2Node_Event}
   n4 = GetPartSkeletalMeshComponent()
   Is Valid(exec=[@n3.then, @n11.then_0], InputObject=@n4)
   > Delay(Duration=0)
   > SetVectorParameterValueOnMaterials!(
    execute
    self:Mesh Component Object Reference=@n4
    ParameterName:Name=headLocationSocket
    ParameterValue:Vector=GetSocketLocation!(self:Scene Component Object Reference=@n4, InSocketName:Name=headLocationSocket) -> (ReturnValue:Vector)
   ) -> (then)
   Event BlueprintUpdateAnimation()
   > n5 = Sequence()
   n6 = Get Body_Mesh()
   > n7 = Set Rotate_ADS(Rotate_ADS:double=SelectFloat(
     A=GetAlphaFromReferenceTransformDelta(SkeletalMeshComponent=@n6, BoneName=head, BaseName=spine_05, DeltaComponentType=QuaternionTwist, InputRange=(X=-48,Y=0), TwistAxis="0,0,1")
     B=0
     bPickA=Get bIsTargeting()
    )) -> (Output_Get:double)
   > n8 = Branch(Condition="float >= float"(A=@n7, B=1) {is=title K2Node_CallFunction})
   > Set Hair_Alpha(Hair_Alpha:double=1) -> (Output_Get:double)
   Set Hair_Alpha(execute=@n8.else, Hair_Alpha=0)
   n9 = Get Body_Mesh()
   Is Valid(exec=@n5.then_1, InputObject=@n9)
   n10 = GetAnimInstance(self=@n9)
   > SetScalarParameterValueOnMaterials!(
    execute
    self:Mesh Component Object Reference=GetOwningComponent()
    ParameterName:Name=Lexa_TransformationCurve
    ParameterValue:float=MapRangeClamped(Value=GetCurveValue(self=@n10, CurveName=Lexa_TransformationCurve), InRangeA=0.5, InRangeB=1)
   ) -> (then)
   n11 = Sequence()
  ]}
]
variables [
 Body_Mesh {type=SkeletalMeshComponent}
 Hair_Alpha {type=number}
 Pawn Owner {type=PlayerPawn_Athena_C}
 Rotate_ADS {type=number}
]
